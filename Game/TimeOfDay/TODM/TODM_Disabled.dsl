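; Time-of-day manager variant that hides its directional light on construction and has an empty DisableLightAndFog event.
; BeginPlay updates scalability, binds that update to user-settings UI changes, and destroys an editor-only post-process workaround component.
; It also writes distance-field availability and the light's shadow distance and fadeout to FortniteMaterialParameters.
; Scalability helpers write global illumination tier flags and reflection quality to that collection, with a GI debug print.
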(event Custom|Update%20Scalabilty%20Settings
  ;[1 1 1] Scalablity MPCs
  (Default|SetGlobalIlluminationMPC)
  (Default|SetReflectionQualityMPC))

(event Custom|DebugScalabilitySettings
  ;[1 1 1] Scalablity MPCs
  (Default|SetGlobalIlluminationMPC)
  (Default|SetReflectionQualityMPC))

(event EventBeginPlay
  (bind _directionallightcomponent (Variables|DayNightCycle|GetDirectionalLightComponent))
  (Utilities|FlowControl|Sequence
    (:then_0
      ;[1 1 1] Scalablity MPCs
      (bind _returnvalue (Settings|GetGameUserSettings))
      (EventDispatchers|BindEventtoOnGameUserSettingsUINeedsUpdate _returnvalue (delegate Custom|Update%20Scalabilty%20Settings))
      (CallFunction|UpdateScalabiltySettings))
    (:then_1
      ;[1 1 1] destroy hacky post process at run time
      (Components|DestroyComponent (Variables|Default|GetEditorPostProcessMaterial(HackToWorkAroundTheOutlineShaderBeingInjectedByCodeAtRuntime)))
      ;[1 1 1] FrontEnd settings
      ;;[1 1 1] Store render settings and make them accessible to shaders. Will need to implement update function when it becomes available.
      (bind _distance_fields_enabled (Utilities|ShadowQualityCheckVFX)
        (:Low
          _)
        (:Medium
          _)
        (:High
          _)
        (:Epic
          _))
      (Rendering|Material|SetScalarParameterValue "/Game/Packages/Fortress_SharedMaterials/GlobalMaterialParameters/FortniteMaterialParameters.FortniteMaterialParameters" "AreDistanceFieldsEnabled?" (Math|Conversions|ToFloat(Boolean) _distance_fields_enabled) :!class /Script/Engine.KismetMaterialLibrary)
      (Rendering|Material|SetScalarParameterValue "/Game/Packages/Fortress_SharedMaterials/GlobalMaterialParameters/FortniteMaterialParameters.FortniteMaterialParameters" "ShadowDistance" (Variables|/Script/Engine.DirectionalLightComponent|GetDynamicShadowDistanceMovableLight _directionallightcomponent) :!class /Script/Engine.KismetMaterialLibrary)
      (Rendering|Material|SetScalarParameterValue "/Game/Packages/Fortress_SharedMaterials/GlobalMaterialParameters/FortniteMaterialParameters.FortniteMaterialParameters" "ShadowDistanceFadeoutFraction" (Variables|/Script/Engine.DirectionalLightComponent|GetShadowDistanceFadeoutFraction _directionallightcomponent) :!class /Script/Engine.KismetMaterialLibrary))))

(event Custom|DisableLightAndFog
  (params (Enable :type Boolean)))

(fn SetGlobalIlluminationMPC
  (bind _returnvalue (Settings|GetGameUserSettings))
  (Rendering|Material|SetScalarParameterValue "/Game/Packages/Fortress_SharedMaterials/GlobalMaterialParameters/FortniteMaterialParameters.FortniteMaterialParameters" "GlobalIlluminationEnabled" (Math|Conversions|ToFloat(Integer) (Math|Integer|Clamp(Integer) (- (Settings|GetGlobalIlluminationQuality _returnvalue) 1) 0 1)) :!class /Script/Engine.KismetMaterialLibrary)
  (Rendering|Material|SetScalarParameterValue "/Game/Packages/Fortress_SharedMaterials/GlobalMaterialParameters/FortniteMaterialParameters.FortniteMaterialParameters" "GlobalIlluminationEpicQuality" (Math|Conversions|ToFloat(Integer) (Math|Integer|Clamp(Integer) (- (Settings|GetGlobalIlluminationQuality _returnvalue) 2) 0 1)) :!class /Script/Engine.KismetMaterialLibrary)
  (Development|PrintString (Utilities|String|Append "GlobalIlluminationDebug " (Utilities|String|ToString(Integer) (Settings|GetGlobalIlluminationQuality _returnvalue)))))

(fn SetReflectionQualityMPC
  (bind _returnvalue (Settings|GetGameUserSettings))
  (Rendering|Material|SetScalarParameterValue "/Game/Packages/Fortress_SharedMaterials/GlobalMaterialParameters/FortniteMaterialParameters.FortniteMaterialParameters" "ReflectionQuality" (Math|Conversions|ToFloat(Integer) (Settings|GetReflectionQuality _returnvalue)) :!class /Script/Engine.KismetMaterialLibrary))

(fn ConstructionScript
  (Rendering|SetVisibility (Variables|DayNightCycle|GetDirectionalLightComponent) false false))
